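# Post-process-style AnimBlueprint for the male medium FurnaceFace body cosmetic, correcting shoulder plates, calves and limb helper bones on top of the parent mesh's copied pose.
# Weighted constraints place the shoulder plate bases and half-calf bones from neighbouring bones.
# Each update maps thigh, calf and upper-arm swing angles to alphas that blend in scale on helper bones.
# Per-side RBF Pose Drivers then drive plate and wishbone bones from arm and clavicle using weapon-pose targets.
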AnimBlueprint /BRCosmetics/Characters/Player/Male/Medium/Bodies/M_MED_FurnaceFace/Meshes/M_MED_Furnace_Face_AnimBP
target_skeleton=&/M_MED_Furnace_Face_Skeleton
graphs [
 AnimGraph {nodes [
   n1 = "Constraint - dyn_shldrPlateBackBase_l (PPPPPPPP)"!(
    ComponentPose=Local To Component(LocalPose=Copy Pose From Mesh() {bUseAttachedParent=true})
    Alpha=1
    ConstraintWeights_0=0.7157
    ConstraintWeights_1:float=0.0209
    ConstraintWeights_2:float=0.0102
    ConstraintWeights_3:float=0.043
    ConstraintWeights_4:float=0.0695
    ConstraintWeights_5:float=0.1074
    ConstraintWeights_6:float=0
    ConstraintWeights_7:float=0.031
   ) -> (Pose) {
    alphaBoolBlend {blendOption=Linear}
    boneToModify {boneName=dyn_shldrPlateBackBase_l}
    constraintSetup [
     {targetBone {boneName=clavicle_l}}
     {targetBone {boneName=upperarm_twist_01_l}}
     {targetBone {boneName=spine_02}}
     {targetBone {boneName=spine_03}}
     {targetBone {boneName=spine_04}}
     {targetBone {boneName=spine_05}}
     {targetBone {boneName=clavicle_r}}
     {targetBone {boneName=upperarm_l}}
    ]
    constraintWeights [1,1,1,1,1,1,1,1]
   }
   n2 = "Constraint - dyn_shldrPlateBackBase_r (PPPPPPPP)"!(
    ComponentPose=@n1
    Alpha=1
    ConstraintWeights_0=0
    ConstraintWeights_1:float=0.0209
    ConstraintWeights_2:float=0.0102
    ConstraintWeights_3:float=0.043
    ConstraintWeights_4:float=0.0695
    ConstraintWeights_5:float=0.1074
    ConstraintWeights_6:float=0.7167
    ConstraintWeights_7:float=0.031
   ) -> (Pose) {
    alphaBoolBlend {blendOption=Linear}
    boneToModify {boneName=dyn_shldrPlateBackBase_r}
    constraintSetup [
     {targetBone {boneName=clavicle_l}}
     {targetBone {boneName=upperarm_twist_01_r}}
     {targetBone {boneName=spine_02}}
     {targetBone {boneName=spine_03}}
     {targetBone {boneName=spine_04}}
     {targetBone {boneName=spine_05}}
     {targetBone {boneName=clavicle_r}}
     {targetBone {boneName=upperarm_r}}
    ]
    constraintWeights [1,1,1,1,1,1,1,1]
   }
   n3 = "Constraint - dyn_shldrPlateFrontBase_l (PPPPPPP)"!(
    ComponentPose=@n2
    Alpha=1
    ConstraintWeights_0=0.5078
    ConstraintWeights_1:float=0
    ConstraintWeights_2:float=0.0012
    ConstraintWeights_3:float=0.0042
    ConstraintWeights_4:float=0.0275
    ConstraintWeights_5:float=0.4492
    ConstraintWeights_6:float=0.0099
   ) -> (Pose) {
    alphaBoolBlend {blendOption=Linear}
    boneToModify {boneName=dyn_shldrPlateFrontBase_l}
    constraintSetup [
     {targetBone {boneName=clavicle_l}}
     {targetBone {boneName=spine_01}}
     {targetBone {boneName=spine_02}}
     {targetBone {boneName=spine_03}}
     {targetBone {boneName=spine_04}}
     {targetBone {boneName=spine_05}}
     {targetBone {boneName=clavicle_r}}
    ]
    constraintWeights [1,1,1,1,1,1,1]
   }
   n4 = "Constraint - dyn_shldrPlateFrontBase_r (PPPPPPPP)"!(
    ComponentPose=@n3
    Alpha=1
    ConstraintWeights_0=0.012
    ConstraintWeights_1:float=0.074
    ConstraintWeights_2:float=0.0052
    ConstraintWeights_3:float=0.0186
    ConstraintWeights_4:float=0.0445
    ConstraintWeights_5:float=0.3349
    ConstraintWeights_6:float=0.4298
    ConstraintWeights_7:float=0.0811
   ) -> (Pose) {
    alphaBoolBlend {blendOption=Linear}
    boneToModify {boneName=dyn_shldrPlateFrontBase_r}
    constraintSetup [
     {targetBone {boneName=clavicle_l}}
     {targetBone {boneName=upperarm_twist_01_r}}
     {targetBone {boneName=spine_02}}
     {targetBone {boneName=spine_03}}
     {targetBone {boneName=spine_04}}
     {targetBone {boneName=spine_05}}
     {targetBone {boneName=clavicle_r}}
     {targetBone {boneName=upperarm_r}}
    ]
    constraintWeights [1,1,1,1,1,1,1,1]
   }
   n5 = "Constraint - dyn_shldrPlateTopBase_r (PPPP)"!(
    ComponentPose=@n4
    Alpha=1
    ConstraintWeights_0=0.4993
    ConstraintWeights_1:float=0.1577
    ConstraintWeights_2:float=0.1884
    ConstraintWeights_3:float=0.1546
   ) -> (Pose) {
    alphaBoolBlend {blendOption=Linear}
    boneToModify {boneName=dyn_shldrPlateTopBase_r}
    constraintSetup [
     {targetBone {boneName=clavicle_r}}
     {targetBone {boneName=upperarm_twist_01_r}}
     {targetBone {boneName=upperarm_r}}
     {targetBone {boneName=spine_05}}
    ]
    constraintWeights [1,1,1,1]
   }
   n6 = "Constraint - dyn_shldrPlateTopBase_l (PPPP)"!(
    ComponentPose=@n5
    Alpha=1
    ConstraintWeights_0=0.4993
    ConstraintWeights_1:float=0.1577
    ConstraintWeights_2:float=0.1884
    ConstraintWeights_3:float=0.1546
   ) -> (Pose) {
    alphaBoolBlend {blendOption=Linear}
    boneToModify {boneName=dyn_shldrPlateTopBase_l}
    constraintSetup [
     {targetBone {boneName=clavicle_l}}
     {targetBone {boneName=upperarm_twist_01_l}}
     {targetBone {boneName=upperarm_l}}
     {targetBone {boneName=spine_05}}
    ]
    constraintWeights [1,1,1,1]
   }
   n7 = "Constraint - dyn_calf_half_l (RR)"!(ComponentPose=@n6, Alpha=1, ConstraintWeights_0=0.1, ConstraintWeights_1:float=0.9) -> (Pose) {
    alphaBoolBlend {blendOption=Linear}
    boneToModify {boneName=dyn_calf_half_l}
    constraintSetup [{transformType=Rotation, targetBone {boneName=thigh_l}}, {transformType=Rotation, targetBone {boneName=calf_l}}]
    constraintWeights [1,1]
   }
   n8 = "Constraint - dyn_calf_half_r (RR)"!(ComponentPose=@n7, Alpha=1, ConstraintWeights_0=0.1, ConstraintWeights_1:float=0.9) -> (Pose) {
    alphaBoolBlend {blendOption=Linear}
    boneToModify {boneName=dyn_calf_half_r}
    constraintSetup [{transformType=Rotation, targetBone {boneName=thigh_r}}, {transformType=Rotation, targetBone {boneName=calf_r}}]
    constraintWeights [1,1]
   }
   n9 = "Transform (Modify) Bone - Bone: thigh_scale_l"(ComponentPose=@n8, Alpha=Get legSwing_l() -> (legSwing_l:double), Translation="0,0,0", Rotation="0,0,0", Scale="16,16,16") {
    scaleMode=BMM_Replace
    scaleSpace=BCS_BoneSpace
    alphaBoolBlend {blendOption=Linear}
    boneToModify {boneName=thigh_scale_l}
    scale=<16 16 16>
   }
   n10 = "Transform (Modify) Bone - Bone: thigh_scale_r"(ComponentPose=@n9, Alpha=Get legSwing_r() -> (legSwing_r:double), Translation="0,0,0", Rotation="0,0,0", Scale="16,16,16") {
    scaleMode=BMM_Replace
    scaleSpace=BCS_BoneSpace
    alphaBoolBlend {blendOption=Linear}
    boneToModify {boneName=thigh_scale_r}
    scale=<16 16 16>
   }
   n11 = "Transform (Modify) Bone - Bone: thigh_low_scale_l"(ComponentPose=@n10, Alpha=Get calfBend_l() -> (calfBend_l:double), Translation="0,0,0", Rotation="0,0,0", Scale="5,5,5") {
    scaleMode=BMM_Replace
    scaleSpace=BCS_BoneSpace
    alphaBoolBlend {blendOption=Linear}
    boneToModify {boneName=thigh_low_scale_l}
    scale=<5 5 5>
   }
   n12 = "Transform (Modify) Bone - Bone: thigh_low_scale_r"(ComponentPose=@n11, Alpha=Get calfBend_r() -> (calfBend_r:double), Translation="0,0,0", Rotation="0,0,0", Scale="5,5,5") {
    scaleMode=BMM_Replace
    scaleSpace=BCS_BoneSpace
    alphaBoolBlend {blendOption=Linear}
    boneToModify {boneName=thigh_low_scale_r}
    scale=<5 5 5>
   }
   n13 = "Transform (Modify) Bone - Bone: upperarm_scale_r"(ComponentPose=@n12, Alpha=Get armSwing_r() -> (armSwing_r:double), Translation="0,0,0", Rotation="0,0,0", Scale="7,7,7") {
    scaleMode=BMM_Replace
    scaleSpace=BCS_BoneSpace
    alphaBoolBlend {blendOption=Linear}
    boneToModify {boneName=upperarm_scale_r}
    scale=<7 7 7>
   }
   n14 = "Transform (Modify) Bone - Bone: upperarm_scale_l"(ComponentPose=@n13, Alpha=Get armSwing_l() -> (armSwing_l:double), Translation="0,0,0", Rotation="0,0,0", Scale="7,7,7") {
    scaleMode=BMM_Replace
    scaleSpace=BCS_BoneSpace
    alphaBoolBlend {blendOption=Linear}
    boneToModify {boneName=upperarm_scale_l}
    scale=<7 7 7>
   }
   n15 = Pose Driver - Source: upperarm_l!(SourcePose:Pose Link Structure=Component To Local(ComponentPose=@n14)) -> (Pose:Pose Link Structure) {
    is=title AnimGraphNode_PoseDriver
    poseAsset=&/M_MED_FurnaceFace_ShoulderPoses_l_PoseAsset
    evalSpaceBone {boneName=spine_05}
    onlyDriveBones [{boneName=dyn_shldrPlateBack_l}, {boneName=dyn_shldrPlateTop_l}, {boneName=dyn_shldrPlateFront_l}, {boneName=dyn_wishbone_l}]
    poseTargets [
     {
      bApplyCustomCurve=false
      bIsHidden=false
      distanceMethod=DefaultMethod
      drivenName=ref_l
      functionType=DefaultFunction
      targetScale=0.7886
      boneTransforms [
       {targetRotation=<rot -48.3229 -172.3356 178.5694>, targetTranslation=<3.3827 -1.0071 -17.7788>}
       {targetRotation=<rot -85.1913 -138.5623 143.0617>, targetTranslation=<4.3575 -0.1466 -2.3223>}
      ]
      customCurve {defaultValue=3.4028234663852886e+38, postInfinityExtrap=RCCE_Constant, preInfinityExtrap=RCCE_Constant, keys []}
      targetRotation=<rot 0 0 0>
     }
     {
      bApplyCustomCurve=false
      bIsHidden=false
      distanceMethod=DefaultMethod
      drivenName=rifle_forward_l
      functionType=DefaultFunction
      targetScale=0
      boneTransforms [
       {targetRotation=<rot -26.9252 149.0419 169.8063>, targetTranslation=<5.2052 1.4451 -17.7304>}
       {targetRotation=<rot -83.3302 61.8918 -57.7668>, targetTranslation=<4.3575 -0.1466 -2.3223>}
      ]
      customCurve {defaultValue=3.4028234663852886e+38, postInfinityExtrap=RCCE_Constant, preInfinityExtrap=RCCE_Constant, keys []}
      targetRotation=<rot 0 0 0>
     }
     {
      bApplyCustomCurve=false
      bIsHidden=false
      distanceMethod=DefaultMethod
      drivenName=rifle_down_l
      functionType=DefaultFunction
      targetScale=0.2968
      boneTransforms [
       {targetRotation=<rot -10.8358 168.1635 -153.7144>, targetTranslation=<3.8542 0.0922 -17.8278>}
       {targetRotation=<rot -87.9627 154.87 -148.4295>, targetTranslation=<4.3575 -0.1466 -2.3223>}
      ]
      customCurve {defaultValue=3.4028234663852886e+38, postInfinityExtrap=RCCE_Constant, preInfinityExtrap=RCCE_Constant, keys []}
      targetRotation=<rot 0 0 0>
     }
     {
      bApplyCustomCurve=false
      bIsHidden=false
      distanceMethod=DefaultMethod
      drivenName=rifle_up_l
      functionType=DefaultFunction
      targetScale=0.9
      boneTransforms [
       {targetRotation=<rot -47.8389 56.2583 -109.5343>, targetTranslation=<14.512 2.5883 -13.7144>}
       {targetRotation=<rot -47.2618 15.0585 -1.3357>, targetTranslation=<4.3575 -0.1466 -2.3223>}
      ]
      customCurve {defaultValue=3.4028234663852886e+38, postInfinityExtrap=RCCE_Constant, preInfinityExtrap=RCCE_Constant, keys []}
      targetRotation=<rot 0 0 0>
     }
     {
      bApplyCustomCurve=false
      bIsHidden=false
      distanceMethod=DefaultMethod
      drivenName=pistol_forward_l
      functionType=DefaultFunction
      targetScale=0
      boneTransforms [
       {targetRotation=<rot -15.3853 125.5033 -167.8701>, targetTranslation=<5.7266 1.1717 -17.716>}
       {targetRotation=<rot -82.9541 43.8452 -40.728>, targetTranslation=<4.3575 -0.1466 -2.3223>}
      ]
      customCurve {defaultValue=3.4028234663852886e+38, postInfinityExtrap=RCCE_Constant, preInfinityExtrap=RCCE_Constant, keys []}
      targetRotation=<rot 0 0 0>
     }
     {
      bApplyCustomCurve=false
      bIsHidden=false
      distanceMethod=DefaultMethod
      drivenName=pistol_down_l
      functionType=DefaultFunction
      targetScale=0.2968
      boneTransforms [
       {targetRotation=<rot 2.9269 169.5881 -152.1165>, targetTranslation=<4.5084 1.029 -17.7882>}
       {targetRotation=<rot -85.6195 82.6791 -80.3659>, targetTranslation=<4.3575 -0.1466 -2.3223>}
      ]
      customCurve {defaultValue=3.4028234663852886e+38, postInfinityExtrap=RCCE_Constant, preInfinityExtrap=RCCE_Constant, keys []}
      targetRotation=<rot 0 0 0>
     }
     {
      bApplyCustomCurve=false
      bIsHidden=false
      distanceMethod=DefaultMethod
      drivenName=pistol_up_l
      functionType=DefaultFunction
      targetScale=0.85
      boneTransforms [
       {targetRotation=<rot -14.9476 65.1736 -140.6509>, targetTranslation=<14.4037 1.5469 -14.0102>}
       {targetRotation=<rot -48.8957 9.558 -1.5714>, targetTranslation=<4.3575 -0.1466 -2.3223>}
      ]
      customCurve {
       defaultValue=3.4028234663852886e+38
       postInfinityExtrap=RCCE_Constant
       preInfinityExtrap=RCCE_Constant
       keys [
        {
         arriveTangent=0
         arriveTangentWeight=0
         interpMode=RCIM_Linear
         leaveTangent=0
         leaveTangentWeight=0
         tangentMode=RCTM_Auto
         tangentWeightMode=RCTWM_WeightedNone
         time=0
         value=0
        }
        {
         arriveTangent=0
         arriveTangentWeight=0
         interpMode=RCIM_Linear
         leaveTangent=0
         leaveTangentWeight=0
         tangentMode=RCTM_Auto
         tangentWeightMode=RCTWM_WeightedNone
         time=1
         value=1
        }
       ]
      }
      targetRotation=<rot 0 0 0>
     }
     {
      bApplyCustomCurve=true
      bIsHidden=false
      distanceMethod=DefaultMethod
      drivenName=rocket_forward_l
      functionType=DefaultFunction
      targetScale=0
      boneTransforms [
       {targetRotation=<rot 8.0899 114.0491 -162.4347>, targetTranslation=<5.6283 3.6245 -17.3087>}
       {targetRotation=<rot -75.1298 71.3911 -71.2642>, targetTranslation=<4.3575 -0.1466 -2.3223>}
      ]
      customCurve {
       defaultValue=3.4028234663852886e+38
       postInfinityExtrap=RCCE_Constant
       preInfinityExtrap=RCCE_Constant
       keys [
        {
         arriveTangent=0
         arriveTangentWeight=0
         interpMode=RCIM_Linear
         leaveTangent=0
         leaveTangentWeight=0
         tangentMode=RCTM_Auto
         tangentWeightMode=RCTWM_WeightedNone
         time=0
         value=0
        }
        {
         arriveTangent=0
         arriveTangentWeight=0
         interpMode=RCIM_Linear
         leaveTangent=0
         leaveTangentWeight=0
         tangentMode=RCTM_Auto
         tangentWeightMode=RCTWM_WeightedNone
         time=1
         value=1
        }
       ]
      }
      targetRotation=<rot 0 0 0>
     }
     {
      bApplyCustomCurve=true
      bIsHidden=false
      distanceMethod=DefaultMethod
      drivenName=rocket_down_l
      functionType=DefaultFunction
      targetScale=0.4258
      boneTransforms [
       {targetRotation=<rot 6.1267 165.1101 -111.7161>, targetTranslation=<4.2487 6.4444 -16.3761>}
       {targetRotation=<rot -64.9297 90.6502 -89.0334>, targetTranslation=<4.3575 -0.1466 -2.3223>}
      ]
      customCurve {
       defaultValue=3.4028234663852886e+38
       postInfinityExtrap=RCCE_Constant
       preInfinityExtrap=RCCE_Constant
       keys [
        {
         arriveTangent=0
         arriveTangentWeight=0
         interpMode=RCIM_Linear
         leaveTangent=0
         leaveTangentWeight=0
         tangentMode=RCTM_Auto
         tangentWeightMode=RCTWM_WeightedNone
         time=0
         value=0
        }
        {
         arriveTangent=0
         arriveTangentWeight=0
         interpMode=RCIM_Linear
         leaveTangent=0
         leaveTangentWeight=0
         tangentMode=RCTM_Auto
         tangentWeightMode=RCTWM_WeightedNone
         time=1
         value=1
        }
       ]
      }
      targetRotation=<rot 0 0 0>
     }
     {
      bApplyCustomCurve=false
      bIsHidden=false
      distanceMethod=DefaultMethod
      drivenName=rocket_up_l
      functionType=DefaultFunction
      targetScale=1
      boneTransforms [
       {targetRotation=<rot 0.3839 23.7878 -123.4787>, targetTranslation=<10.348 5.423 -15.5032>}
       {targetRotation=<rot -58.1814 42.9306 -47.4999>, targetTranslation=<4.3575 -0.1466 -2.3223>}
      ]
      customCurve {
       defaultValue=3.4028234663852886e+38
       postInfinityExtrap=RCCE_Constant
       preInfinityExtrap=RCCE_Constant
       keys [
        {
         arriveTangent=0
         arriveTangentWeight=0
         interpMode=RCIM_Linear
         leaveTangent=0
         leaveTangentWeight=0
         tangentMode=RCTM_Auto
         tangentWeightMode=RCTWM_WeightedNone
         time=0
         value=0
        }
        {
         arriveTangent=0
         arriveTangentWeight=0
         interpMode=RCIM_Linear
         leaveTangent=0
         leaveTangentWeight=0
         tangentMode=RCTM_Auto
         tangentWeightMode=RCTWM_WeightedNone
         time=1
         value=1
        }
       ]
      }
      targetRotation=<rot 0 0 0>
     }
     {
      bApplyCustomCurve=true
      bIsHidden=false
      distanceMethod=DefaultMethod
      drivenName=mini_forward_l
      functionType=Cubic
      targetScale=0
      boneTransforms [
       {targetRotation=<rot -8.1761 143.5074 -123.6052>, targetTranslation=<5.0907 3.7287 -17.3237>}
       {targetRotation=<rot -75.2704 79.2922 -75.4184>, targetTranslation=<4.3575 -0.1466 -2.3223>}
      ]
      customCurve {
       defaultValue=3.4028234663852886e+38
       postInfinityExtrap=RCCE_Constant
       preInfinityExtrap=RCCE_Constant
       keys [
        {
         arriveTangent=0
         arriveTangentWeight=0
         interpMode=RCIM_Linear
         leaveTangent=0
         leaveTangentWeight=0
         tangentMode=RCTM_Auto
         tangentWeightMode=RCTWM_WeightedNone
         time=0
         value=0
        }
        {
         arriveTangent=0
         arriveTangentWeight=0
         interpMode=RCIM_Linear
         leaveTangent=0
         leaveTangentWeight=0
         tangentMode=RCTM_Auto
         tangentWeightMode=RCTWM_WeightedNone
         time=0.2836
         value=0.6667
        }
        {
         arriveTangent=0
         arriveTangentWeight=0
         interpMode=RCIM_Linear
         leaveTangent=0
         leaveTangentWeight=0
         tangentMode=RCTM_Auto
         tangentWeightMode=RCTWM_WeightedNone
         time=1
         value=1
        }
       ]
      }
      targetRotation=<rot 0 0 0>
     }
     {
      bApplyCustomCurve=false
      bIsHidden=false
      distanceMethod=DefaultMethod
      drivenName=mini_down_l
      functionType=DefaultFunction
      targetScale=0.4125
      boneTransforms [
       {targetRotation=<rot -12.5077 163.2372 -153.2284>, targetTranslation=<2.3344 5.0159 -16.8208>}
       {targetRotation=<rot -69.137 111.0334 -104.2001>, targetTranslation=<4.3575 -0.1466 -2.3223>}
      ]
      customCurve {defaultValue=3.4028234663852886e+38, postInfinityExtrap=RCCE_Constant, preInfinityExtrap=RCCE_Constant, keys []}
      targetRotation=<rot 0 0 0>
     }
     {
      bApplyCustomCurve=false
      bIsHidden=false
      distanceMethod=DefaultMethod
      drivenName=mini_up_l
      functionType=DefaultFunction
      targetScale=0.7421
      boneTransforms [
       {targetRotation=<rot -27.0194 83.5884 -117.5376>, targetTranslation=<13.2007 -6.2306 -13.5167>}
       {targetRotation=<rot -46.1813 -34.536 22.3519>, targetTranslation=<4.3575 -0.1466 -2.3223>}
      ]
      customCurve {defaultValue=3.4028234663852886e+38, postInfinityExtrap=RCCE_Constant, preInfinityExtrap=RCCE_Constant, keys []}
      targetRotation=<rot 0 0 0>
     }
    ]
    rBFParams {radius=24.368}
    sourceBones [{boneName=upperarm_l}, {boneName=clavicle_l}]
   }
   n16 = Pose Driver - Source: upperarm_r!(SourcePose:Pose Link Structure=@n15) -> (Pose:Pose Link Structure) {
    is=title AnimGraphNode_PoseDriver
    poseAsset=&/M_MED_FurnaceFace_ShoulderPoses_r_PoseAsset
    evalSpaceBone {boneName=spine_05}
    onlyDriveBones [{boneName=dyn_shldrPlateBack_r}, {boneName=dyn_shldrPlateTop_r}, {boneName=dyn_shldrPlateFront_r}, {boneName=dyn_wishbone_r}]
    poseTargets [
     {
      bApplyCustomCurve=false
      bIsHidden=false
      distanceMethod=DefaultMethod
      drivenName=ref_r
      functionType=DefaultFunction
      targetScale=0.4835
      boneTransforms [
       {targetRotation=<rot -48.3229 7.6644 178.5694>, targetTranslation=<3.3828 -1.007 17.7788>}
       {targetRotation=<rot -85.1913 41.4375 143.0617>, targetTranslation=<4.3575 -0.1466 2.3223>}
      ]
      customCurve {defaultValue=3.4028234663852886e+38, postInfinityExtrap=RCCE_Constant, preInfinityExtrap=RCCE_Constant, keys []}
      targetRotation=<rot 0 0 0>
     }
     {
      bApplyCustomCurve=false
      bIsHidden=false
      distanceMethod=DefaultMethod
      drivenName=rifle_forward_r
      functionType=DefaultFunction
      targetScale=0
      boneTransforms [
       {targetRotation=<rot 2.0313 -30.6347 -122.7273>, targetTranslation=<7.5132 2.7742 17.2256>}
       {targetRotation=<rot -73.894 -137.2169 -42.5933>, targetTranslation=<4.3575 -0.1466 2.3223>}
      ]
      customCurve {defaultValue=3.4028234663852886e+38, postInfinityExtrap=RCCE_Constant, preInfinityExtrap=RCCE_Constant, keys []}
      targetRotation=<rot 0 0 0>
     }
     {
      bApplyCustomCurve=false
      bIsHidden=false
      distanceMethod=DefaultMethod
      drivenName=rifle_down_r
      functionType=DefaultFunction
      targetScale=0.89
      boneTransforms [
       {targetRotation=<rot -15.928 -15.6403 -116.2718>, targetTranslation=<11.7343 2.9859 15.6028>}
       {targetRotation=<rot -58.8671 -156.9802 -24.0258>, targetTranslation=<4.3575 -0.1466 2.3223>}
      ]
      customCurve {defaultValue=3.4028234663852886e+38, postInfinityExtrap=RCCE_Constant, preInfinityExtrap=RCCE_Constant, keys []}
      targetRotation=<rot 0 0 0>
     }
     {
      bApplyCustomCurve=false
      bIsHidden=false
      distanceMethod=DefaultMethod
      drivenName=rifle_up_r
      functionType=DefaultFunction
      targetScale=0
      boneTransforms [
       {targetRotation=<rot 17.8984 -42.6976 -129.7035>, targetTranslation=<7.0991 2.767 17.3085>}
       {targetRotation=<rot -75.0565 -133.2485 -46.44>, targetTranslation=<4.3575 -0.1466 2.3223>}
      ]
      customCurve {defaultValue=3.4028234663852886e+38, postInfinityExtrap=RCCE_Constant, preInfinityExtrap=RCCE_Constant, keys []}
      targetRotation=<rot 0 0 0>
     }
     {
      bApplyCustomCurve=false
      bIsHidden=false
      distanceMethod=DefaultMethod
      drivenName=pistol_forward_r
      functionType=DefaultFunction
      targetScale=0
      boneTransforms [
       {targetRotation=<rot 27.7556 -78.5482 -130.4284>, targetTranslation=<7.9754 4.0136 16.8201>}
       {targetRotation=<rot -69.1784 -131.0124 -50.2075>, targetTranslation=<4.3575 -0.1466 2.3223>}
      ]
      customCurve {defaultValue=3.4028234663852886e+38, postInfinityExtrap=RCCE_Constant, preInfinityExtrap=RCCE_Constant, keys []}
      targetRotation=<rot 0 0 0>
     }
     {
      bApplyCustomCurve=false
      bIsHidden=false
      distanceMethod=DefaultMethod
      drivenName=pistol_down_r
      functionType=DefaultFunction
      targetScale=1
      boneTransforms [
       {targetRotation=<rot 14.9502 -35.1574 -102.1951>, targetTranslation=<6.7377 4.2374 17.0092>}
       {targetRotation=<rot -71.238 -118.5013 -62.2217>, targetTranslation=<4.3575 -0.1466 2.3223>}
      ]
      customCurve {defaultValue=3.4028234663852886e+38, postInfinityExtrap=RCCE_Constant, preInfinityExtrap=RCCE_Constant, keys []}
      targetRotation=<rot 0 0 0>
     }
     {
      bApplyCustomCurve=false
      bIsHidden=false
      distanceMethod=DefaultMethod
      drivenName=pistol_up_r
      functionType=DefaultFunction
      targetScale=0.7565
      boneTransforms [
       {targetRotation=<rot 22.8073 -125.5346 -177.6895>, targetTranslation=<7.7489 4.2571 16.8024>}
       {targetRotation=<rot -68.9934 -127.6294 -53.9111>, targetTranslation=<4.3575 -0.1466 2.3223>}
      ]
      customCurve {defaultValue=3.4028234663852886e+38, postInfinityExtrap=RCCE_Constant, preInfinityExtrap=RCCE_Constant, keys []}
      targetRotation=<rot 0 0 0>
     }
     {
      bApplyCustomCurve=false
      bIsHidden=false
      distanceMethod=DefaultMethod
      drivenName=rocket_forward_r
      functionType=DefaultFunction
      targetScale=0
      boneTransforms [
       {targetRotation=<rot -17.6773 -42.0746 -142.9696>, targetTranslation=<8.2801 -1.3003 17.2799>}
       {targetRotation=<rot -74.713 163.6034 18.7452>, targetTranslation=<4.3575 -0.1466 2.3223>}
      ]
      customCurve {defaultValue=3.4028234663852886e+38, postInfinityExtrap=RCCE_Constant, preInfinityExtrap=RCCE_Constant, keys []}
      targetRotation=<rot 0 0 0>
     }
     {
      bApplyCustomCurve=false
      bIsHidden=false
      distanceMethod=DefaultMethod
      drivenName=rocket_down_r
      functionType=DefaultFunction
      targetScale=0.3294
      boneTransforms [
       {targetRotation=<rot -31.845 14.7319 -152.2645>, targetTranslation=<8.9489 -1.7552 17.0624>}
       {targetRotation=<rot -71.8091 160.79 19.5509>, targetTranslation=<4.3575 -0.1466 2.3223>}
      ]
      customCurve {defaultValue=3.4028234663852886e+38, postInfinityExtrap=RCCE_Constant, preInfinityExtrap=RCCE_Constant, keys []}
      targetRotation=<rot 0 0 0>
     }
     {
      bApplyCustomCurve=false
      bIsHidden=false
      distanceMethod=DefaultMethod
      drivenName=rocket_up_r
      functionType=DefaultFunction
      targetScale=0.7562
      boneTransforms [
       {targetRotation=<rot -3.749 -99.0927 -155.7327>, targetTranslation=<8.0897 -1.5047 17.3159>}
       {targetRotation=<rot -75.1461 160.0167 22.6425>, targetTranslation=<4.3575 -0.1466 2.3223>}
      ]
      customCurve {defaultValue=3.4028234663852886e+38, postInfinityExtrap=RCCE_Constant, preInfinityExtrap=RCCE_Constant, keys []}
      targetRotation=<rot 0 0 0>
     }
     {
      bApplyCustomCurve=false
      bIsHidden=false
      distanceMethod=DefaultMethod
      drivenName=mini_forward_r
      functionType=DefaultFunction
      targetScale=0
      boneTransforms [
       {targetRotation=<rot -30.8418 34.4647 -161.168>, targetTranslation=<8.7615 -3.7849 16.743>}
       {targetRotation=<rot -68.3884 140.4413 46.6652>, targetTranslation=<4.3575 -0.1466 2.3223>}
      ]
      customCurve {defaultValue=3.4028234663852886e+38, postInfinityExtrap=RCCE_Constant, preInfinityExtrap=RCCE_Constant, keys []}
      targetRotation=<rot 0 0 0>
     }
     {
      bApplyCustomCurve=false
      bIsHidden=false
      distanceMethod=DefaultMethod
      drivenName=mini_down_r
      functionType=DefaultFunction
      targetScale=1.25
      boneTransforms [
       {targetRotation=<rot -76.4655 -6.9355 -126.1384>, targetTranslation=<16.2186 -1.6873 12.2257>}
       {targetRotation=<rot -39.6638 172.6056 16.7783>, targetTranslation=<4.3575 -0.1466 2.3223>}
      ]
      customCurve {defaultValue=3.4028234663852886e+38, postInfinityExtrap=RCCE_Constant, preInfinityExtrap=RCCE_Constant, keys []}
      targetRotation=<rot 0 0 0>
     }
     {
      bApplyCustomCurve=false
      bIsHidden=false
      distanceMethod=DefaultMethod
      drivenName=mini_up_r
      functionType=DefaultFunction
      targetScale=0.5247
      boneTransforms [
       {targetRotation=<rot -16.3953 3.5187 -131.537>, targetTranslation=<5.2835 3.2349 17.4328>}
       {targetRotation=<rot -76.8956 -105.4544 -66.2493>, targetTranslation=<4.3575 -0.1466 2.3223>}
      ]
      customCurve {defaultValue=3.4028234663852886e+38, postInfinityExtrap=RCCE_Constant, preInfinityExtrap=RCCE_Constant, keys []}
      targetRotation=<rot 0 0 0>
     }
    ]
    rBFParams {radius=28.1164}
    sourceBones [{boneName=upperarm_r}, {boneName=clavicle_r}]
   }
   Output Pose(Result=@n16) {layerGroup=None}
  ]}
 EventGraph {nodes [
   Event BlueprintUpdateAnimation()
   n1 = GetOwnerSkeletalMeshComponent()
   > Set legSwing_r(legSwing_r:double=MapRangeClamped(Value=GetTwistAndSwingAngleOfDeltaRotationFromRefPose(self=@n1, BoneName=thigh_r).OutSwingAngle, InRangeA=175, InRangeB=71)) -> (Output_Get:double)
   > Set legSwing_l(legSwing_l:double=MapRangeClamped(Value=GetTwistAndSwingAngleOfDeltaRotationFromRefPose(self=@n1, BoneName=thigh_l).OutSwingAngle, InRangeA=35, InRangeB=101)) -> (Output_Get:double)
   > Set armSwing_r(armSwing_r:double=MapRangeClamped(Value=GetTwistAndSwingAngleOfDeltaRotationFromRefPose(self=@n1, BoneName=upperarm_r).OutSwingAngle, InRangeA=55, InRangeB=118)) -> (Output_Get:double)
   > Set calfBend_r(calfBend_r:double=MapRangeClamped(Value=GetTwistAndSwingAngleOfDeltaRotationFromRefPose(self=@n1, BoneName=calf_r).OutSwingAngle, InRangeA=23, InRangeB=118)) -> (Output_Get:double)
   > Set calfBend_l(calfBend_l:double=MapRangeClamped(Value=GetTwistAndSwingAngleOfDeltaRotationFromRefPose(self=@n1, BoneName=calf_l).OutSwingAngle, InRangeA=23, InRangeB=118)) -> (Output_Get:double)
   > Set armSwing_l(armSwing_l:double=MapRangeClamped(Value=GetTwistAndSwingAngleOfDeltaRotationFromRefPose(self=@n1, BoneName=upperarm_l).OutSwingAngle, InRangeA=45, InRangeB=110)) -> (Output_Get:double)
   TryGetPawnOwner()
  ]}
]
variables [
 armSwing_l {type=number}
 armSwing_r {type=number}
 calfBend_l {type=number}
 calfBend_r {type=number}
 legSwing_l {type=number}
 legSwing_r {type=number}
]
